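# Animation Blueprint for the female medium ScubaDasher body cosmetic, adding secondary motion to its accessories.
# It copies the attached parent's pose, runs a Control Rig, then weights dyn_rope_01 between pelvis and spine bones.
# Chained AnimDynamics nodes spring the pouch, knife, zipper and lower rope bones, with spherical limits on the rope.
# Each update computes a gravity override from the M_MED (male-prefixed) parameter asset, fed only to the knife node.
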AnimBlueprint /BRCosmetics/Characters/Player/Female/Medium/Bodies/F_MED_ScubaDasher/Meshes/F_MED_ScubaDasher_AnimBP
target_skeleton=&/F_MED_ScubaDasher_Skeleton
graphs [
 AnimGraph {nodes [
   n1 = Control Rig(Source=Copy Pose From Mesh() {bUseAttachedParent=true}) {
    alphaBoolBlend {blendOption=Linear}
    controlRigAssetReference {blueprintRigClass=&/F_MED_ScubaDasher_CtrlRig.F_MED_ScubaDasher_CtrlRig_C}
   }
   n2 = "Constraint - dyn_rope_01 (PPP)"!(
    ComponentPose=Local To Component(LocalPose=@n1)
    Alpha=1
    ConstraintWeights_0=0.7068
    ConstraintWeights_1:float=0.2383
    ConstraintWeights_2:float=0.0549
   ) -> (Pose) {
    alphaBoolBlend {blendOption=Linear}
    boneToModify {boneName=dyn_rope_01}
    constraintSetup [{targetBone {boneName=pelvis}}, {targetBone {boneName=spine_02}}, {targetBone {boneName=spine_03}}]
    constraintWeights [1,1,1]
   }
   n3 = Anim Dynamics - Bone: dyn_Pouch_l(ComponentPose=@n2, Alpha=1, ExternalForce="0,0,0") {
    angularDampingOverride=1
    angularSpringConstant=3000
    bAngularSpring=true
    bOverrideAngularDamping=true
    bUseGravityOverride=true
    gravityScale=0.35
    simulationSpace=RootRelative
    alphaBoolBlend {blendOption=Linear}
    boundBone {boneName=dyn_Pouch_l}
    componentLinearAccScale=<1 1 1>
    componentLinearVelScale=<1 1 1>
    gravityOverride {z=-980}
    physicsBodyDefinitions [{
      boundBone {boneName=dyn_Pouch_l}
      boxExtents=<7 10 10>
      constraintSetup {angularLimitsMax=<20 0 0>, angularLimitsMin=<-20 0 0>, angularTarget=<1 0 0>}
      localJointOffset=<1.5 0 0>
     }]
   }
   n4 = Anim Dynamics - Bone: dyn_knife_r!(ComponentPose=@n3, Alpha=1, GravityOverride:Vector=Get Gravity Override(), ExternalForce="0,0,0") -> (Pose) {
    angularDampingOverride=1
    angularSpringConstant=200
    bAngularSpring=true
    bOverrideAngularDamping=true
    bUseGravityOverride=true
    gravityScale=0.35
    simulationSpace=RootRelative
    alphaBoolBlend {blendOption=Linear}
    boundBone {boneName=dyn_knife_r}
    componentLinearAccScale=<1 1 1>
    componentLinearVelScale=<1 1 1>
    gravityOverride {z=-980}
    physicsBodyDefinitions [{
      boundBone {boneName=dyn_knife_r}
      boxExtents=<6 11 11>
      constraintSetup {angularLimitsMax=<0 2 3>, angularLimitsMin=<0 -2 -3>, angularTarget=<1 0 0>}
      localJointOffset=<1 0 0>
     }]
   }
   n5 = Anim Dynamics - Bone: dyn_zipper(ComponentPose=@n4, Alpha=1) {
    angularDampingOverride=1
    angularSpringConstant=450
    bAngularSpring=true
    bOverrideAngularDamping=true
    bUseGravityOverride=true
    gravityScale=0.35
    simulationSpace=RootRelative
    alphaBoolBlend {blendOption=Linear}
    boundBone {boneName=dyn_zipper}
    componentLinearAccScale=<1 1 1>
    componentLinearVelScale=<1 1 1>
    gravityOverride {z=-980}
    physicsBodyDefinitions [{
      boundBone {boneName=dyn_zipper}
      boxExtents=<5 10 10>
      constraintSetup {angularLimitsMax=<10 0 10>, angularLimitsMin=<-10 -60 -10>, angularTarget=<1 0 0>}
      localJointOffset=<2.5 0 0>
     }]
   }
   n6 = Anim Dynamics - Bone: dyn_rope_02(ComponentPose=@n5, Alpha=1, ExternalForce="0,0,0") {
    angularDampingOverride=1
    angularSpringConstant=3000
    bAngularSpring=true
    bOverrideAngularDamping=true
    bUseGravityOverride=true
    bUseSphericalLimits=true
    gravityScale=0.35
    simulationSpace=RootRelative
    alphaBoolBlend {blendOption=Linear}
    boundBone {boneName=dyn_rope_02}
    componentLinearAccScale=<0.5 0.5 0.5>
    componentLinearVelScale=<0.5 0.5 0.5>
    gravityOverride {z=-980}
    physicsBodyDefinitions [{
      boundBone {boneName=dyn_rope_02}
      boxExtents=<10 15 15>
      constraintSetup {angularLimitsMax=<0 0 30>, angularLimitsMin=<0 -50 -30>, angularTarget=<1 0 0>}
      localJointOffset=<3 0 0>
     }]
    sphericalLimits [
     {limitRadius=8.9, drivingBone {boneName=pelvis}, sphereLocalOffset=<0 0 0>}
     {limitRadius=9, drivingBone {boneName=None}, sphereLocalOffset=<5.3 0 0>}
    ]
   }
   n7 = Anim Dynamics - Bone: dyn_rope_03(ComponentPose=@n6, Alpha=1, ExternalForce="0,0,0") {
    angularDampingOverride=1
    angularSpringConstant=3000
    bAngularSpring=true
    bOverrideAngularDamping=true
    bUseGravityOverride=true
    bUseSphericalLimits=true
    gravityScale=0.35
    simulationSpace=RootRelative
    alphaBoolBlend {blendOption=Linear}
    boundBone {boneName=dyn_rope_03}
    componentLinearAccScale=<0.5 0.5 0.5>
    componentLinearVelScale=<0.5 0.5 0.5>
    gravityOverride {z=-980}
    physicsBodyDefinitions [{
      boundBone {boneName=dyn_rope_03}
      boxExtents=<10 15 15>
      constraintSetup {angularLimitsMax=<0 0 30>, angularLimitsMin=<0 -50 -30>, angularTarget=<1 0 0>}
      localJointOffset=<2 0 0>
     }]
    sphericalLimits [{limitRadius=10, drivingBone {boneName=thigh_r}, sphereLocalOffset=<6 0 0>}]
   }
   Output Pose(Result=Component To Local(ComponentPose=@n7)) {layerGroup=None}
  ]}
 EventGraph {nodes [
   Event BlueprintUpdateAnimation()
   > n1 = "Gravity Override State Function (Parameter Asset) (Nativized)"(
    CharacterPartAnimInstance=Self-Reference()
    GameTimeInSeconds=GetGameTimeInSeconds()
    Parameters=&/M_MED_ScubaDasher_Parameters
   )
   > Set Gravity Override(Gravity Override=@n1)
  ]}
]
variables [
 Gravity Override {type=Vector}
 Gravity Override Params {type=GravityOverrideParamsStruct}
]
